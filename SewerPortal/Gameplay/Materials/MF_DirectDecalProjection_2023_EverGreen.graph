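# Material function that projects a graffiti decal texture onto the Base Color of incoming material attributes, toggled by a static switch.
# Pre-skinned position, offset and rotated by matrix parameters, becomes tiled planar UVs, so the decal follows deforming meshes.
# A square SDF edge fade, a normal-versus-projection-vector angle blend and layer enables selected by UV1 shape the decal's alpha.
# Packed vector parameters control mask shape, blending, alpha multiplier and a debug mask view.
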MaterialFunction /SewerPortal/Gameplay/Materials/MF_DirectDecalProjection_2023_EverGreen
graph {nodes [
  n1 = DotProduct(
   A=Normalize(VectorInput=VectorParameter() {group=Graffiti, parameterName=ProjectVector, defaultValue=<rgba 0 0 -1 1>}.RGB)
   B=VertexNormalWS()
  )
  ConstantBiasScale(None=@n1)
  Constant() {r=1}
  n2 = FunctionInput(Preview=SetMaterialAttributes!(MaterialAttributes, BaseColor=Constant() {r=1}) {attributeSetTypes [#1]}) {inputName=MA, inputType=FunctionInput_MaterialAttributes}
  n3 = GetMaterialAttributes(None=@n2) {attributeGetTypes [#1]}
  n4 = TextureSampleParameter2D(UVs=NamedRerouteUsage()) {group=Graffiti, parameterName=Diffuse_ProjectedDecal, texture=&T_BlacksRGB}
  n5 = NamedRerouteUsage()
  n6 = Multiply(A=@n4.A, B=NamedRerouteUsage())
  n7 = Multiply(A=Multiply(A=Saturate(None=@n6), B=NamedRerouteUsage()), B=Multiply(A=NamedRerouteUsage(), B=NamedRerouteUsage()) {constA=1})
  n8 = LinearInterpolate(A=@n3.BaseColor, B=LinearInterpolate(A=@n4.RGB, B=@n7, Alpha=@n5), Alpha=LinearInterpolate(A=@n7, B=NamedRerouteUsage(), Alpha=@n5))
  FunctionOutput(None=StaticSwitchParameter(
    True=SetMaterialAttributes(MaterialAttributes=@n3.MaterialAttributes, BaseColor=@n8) {attributeSetTypes [#1]}
    False=@n2
   ) {group=Graffiti, parameterName=Use Projected Decal}) {sortPriority=10}
  n9 = PreSkinnedPosition()
  n10 = VectorParameter() {group=Graffiti, parameterName=Matrix_X, channelNames {a=XOffset, b="", g="", r=""}, defaultValue=<rgba 1 0 0 0>}
  n11 = VectorParameter() {group=Graffiti, parameterName=Matrix_Y, channelNames {a=YOffset, b="", g="", r=""}, defaultValue=<rgba 0 1 0 0>}
  n12 = VectorParameter() {group=Graffiti, parameterName=Matrix_Z, channelNames {a=ZOffset, b="", g="", r=""}, defaultValue=<rgba 0 0 1 0>}
  n13 = Add(
   A=ShaderStageSwitch(PixelShader=VertexInterpolator(VS=@n9).PS, VertexShader=@n9)
   B=MaterialFunctionCall(X=@n10.XOffset, Y=@n11.YOffset, Z=@n12.ZOffset) {materialFunction=&MakeFloat3}.Result
  )
  n14 = Multiply(
   A=ComponentMask(None=MaterialFunctionCall(VectorToTransform=@n13, BasisX=@n10, BasisY=@n11, BasisZ=@n12) {materialFunction=&Transform3x3Matrix}.Transformed Vector) {g=true, r=true}
   B=ScalarParameter() {defaultValue=0.02, group=Graffiti, parameterName=Tiling}
  )
  n15 = NamedRerouteDeclaration(None=@n14) {name=UVs_Centered, nodeColor=<rgba 0 1 0.25 1>}
  n16 = Constant2Vector() {g=0.5, r=0.5}
  NamedRerouteDeclaration(None=Multiply(A=Add(A=@n15, B=@n16)) {constB=-1}) {name=UVs, nodeColor=<rgba 0 1 0.25 1>}
  n17 = MaterialFunctionCall(Float2=SmoothStep(Value=Abs(None=@n14)) {constMax=0.5, constMin=0.5}) {materialFunction=&BreakOutFloat2Components}
  n18 = OneMinus(None=Max(A=@n17.R, B=@n17.R))
  n19 = NamedRerouteUsage()
  NamedRerouteDeclaration(None=Multiply(A=Multiply(A=@n18, B=@n19), B=NamedRerouteUsage())) {name=Mask, nodeColor=<rgba 0.7578 1 0 1>}
  n20 = LinearInterpolate(Alpha=NamedRerouteUsage()) {constB=0.5}
  n21 = NamedRerouteUsage()
  n22 = LinearInterpolate(Alpha=NamedRerouteUsage()) {constA=-0.5, constB=0}
  n23 = Add(A=@n22) {constB=0.5}
  n24 = Subtract(
   A=Subtract(A=Abs(None=@n21)) {constB=0.5, desc=CompWiseEdgeDist}
   B=LinearInterpolate(A=@n22, B=@n23, Alpha=NamedRerouteUsage()) {constB=0.5}
  )
  n25 = Max(A=@n24) {constB=0}
  n26 = MaterialFunctionCall(Float2=@n24) {materialFunction=&BreakOutFloat2Components}
  n27 = Max(A=@n26.R, B=@n26.R)
  n28 = Add(
   A=Add(A=Length(None=@n25) {is=class MaterialExpressionLength}, B=Min(A=@n27) {constB=0}) {desc=Square SDF}
   B=LinearInterpolate(A=@n23, B=Add(A=@n22), Alpha=NamedRerouteUsage()) {constA=0.5}
  )
  NamedRerouteDeclaration(None=OneMinus(None=SmoothStep(Min=@n20, Value=@n28) {constMax=0.5})) {name=Edge Fade, nodeColor=<rgba 0.0781 0 1 1>}
  n29 = VectorParameter() {
   group=Graffiti
   parameterName=DD_Params
   channelNames {a=Angle Hardness, b=Angle Blend, g=Mask Blend, r=Mask Shape}
   defaultValue=<rgba 1 1 0.5 1>
  }
  NamedRerouteDeclaration(None=@n29.Mask Shape) {name=Mask Shape Param, nodeColor=<rgba 0.8516 1 0 1>}
  NamedRerouteDeclaration(None=@n29.Mask Blend) {name=Mask Blend Param, nodeColor=<rgba 0.8516 1 0 1>}
  NamedRerouteDeclaration(None=@n29.Angle Blend) {name=Angle Blend Param, nodeColor=<rgba 0.8516 1 1 1>}
  NamedRerouteDeclaration(None=@n29.Angle Hardness) {name=Angle Hardness Param, nodeColor=<rgba 0.8516 1 0 1>}
  n30 = ComponentMask(None=TextureCoordinate() {coordinateIndex=1}) {r=true}
  n31 = VectorParameter() {group=Graffiti, parameterName=DD_EnableDirectDecalLayer, channelNames {a=Layer4, b=Layer3, g=Layer2, r=Layer1}}
  n32 = Constant() {r=2}
  n33 = If(A=Constant() {r=3}, B=@n30, "A > B"=@n31.Layer3, "A == B"=@n31.Layer3, "A < B"=@n31.Layer3)
  NamedRerouteDeclaration(None=If(
    A=Constant() {r=1}
    B=@n30
    "A > B"=@n31.Layer1
    "A == B"=@n31.Layer1
    "A < B"=If(A=@n32, B=@n30, "A > B"=@n31.Layer2, "A == B"=@n31.Layer2, "A < B"=@n33)
   ) {desc=0}) {name=Enabled Layers, nodeColor=<rgba 1 0 0.0469 1>}
  n34 = NamedRerouteUsage()
  NamedRerouteDeclaration(None=SphereMask(A=@n1, B=Constant() {r=-1}, Radius=LinearInterpolate(Alpha=@n34) {constAlpha=1, constB=2.01}, Hardness=NamedRerouteUsage()) {attenuationRadius=2, hardnessPercent=50}) {name=Angle Blend, nodeColor=<rgba 1 0 0.3281 1>}
  n35 = VectorParameter() {group=Graffiti, parameterName=DD_Params2, channelNames {a="", b="", g=Texture Alpha Multiplier, r=Debug Mask}, defaultValue=<rgba 0 1 0 1>}
  NamedRerouteDeclaration(None=@n35.Debug Mask) {name=Debug Mask, nodeColor=<rgba 0 0.6953 1 1>}
  NamedRerouteDeclaration(None=@n35.Texture Alpha Multiplier) {name=Alpha Multiplier, nodeColor=<rgba 0 0.2969 1 1>}
  n36 = TransformPosition(None=PreSkinnedNormal()) {transformType=TRANSFORMPOSSOURCE_World}
  ShaderStageSwitch(PixelShader=VertexInterpolator(VS=@n36).PS, VertexShader=@n36)
  TransformPosition(None=@n9) {transformType=TRANSFORMPOSSOURCE_World}
 ]}
