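# Material function that animates a sewer portal lid opening through World Position Offset, with matching corrected normals.
# A curve atlas remaps custom-primitive-data progress, or looping preview time, into offset-rotation, height, self-rotation and damping channels.
# Self-rotation and offset-pivot rotation with an extra Z-matrix spin are switched by Ceil-driven lerps, then height and pivot offsets are added.
# The graph also holds a duplicate, unconnected copy of this chain.
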MaterialFunction /SewerPortal/Gameplay/Materials/MF_LidAnimation_EverGreen
graph {nodes [
  n1 = Transform(None=VectorParameter() {group=LidTransform, parameterName=LocalRotationAxisFinal, defaultValue=<rgba 1 0 0 1>}.RGB) {transformSourceType=TRANSFORMSOURCE_Local}
  n2 = NamedRerouteUsage()
  n3 = ScalarParameter() {defaultValue=-0.24, group=LidTransform, parameterName=OffsetRadian}
  n4 = VectorParameter() {group=LidTransform, parameterName=XRotatePivot, defaultValue=<rgba 0 50 0 1>}
  n5 = MaterialFunctionCall() {materialFunction=&ObjectPivotPoint}
  n6 = WorldPosition()
  n7 = Add(
   A=RotateAboutAxis(
    NormalizedRotationAxis=@n1
    RotationAngle=Multiply(A=@n2, B=@n3) {constB=-1}
    PivotPoint=Add(A=@n4.RGB, B=@n5.Object Pivot Location)
    Position=@n6.XYZ
   )
   B=WorldPosition().XYZ
  )
  n8 = NamedRerouteUsage()
  Composite!(None=@n7, "BasisPosition (V3)"=@n5.Object Pivot Location, Input2=@n8, Input3=@n8) {subgraphName=SameRotateTogether}
  n9 = GetMaterialAttributes(None=NamedRerouteDeclaration(None=FunctionInput() {inputName=Material, inputType=FunctionInput_MaterialAttributes}) {name=MA_IN, nodeColor=<rgba 0.7578 1 0 1>}) {attributeGetTypes [#1]}
  n10 = MaterialFunctionCall(Rotation Axis=NamedRerouteUsage(), Rotation Angle=NamedRerouteUsage()) {materialFunction=&FixRotateAboutAxisNormals}
  n11 = NamedRerouteUsage()
  n12 = Cosine(None=@n11)
  n13 = Sine(None=@n11)
  n14 = Constant()
  n15 = MaterialFunctionCall(
   VectorToTransform=Normalize(VectorInput=Transform(None=@n10.New Tangent Space Vertex Normal))
   BasisX=MaterialFunctionCall(X=@n12, Y=Multiply(A=@n13) {constB=-1}, Z=@n14) {materialFunction=&MakeFloat3}.Result
   BasisY=MaterialFunctionCall(X=@n13, Y=@n12, Z=@n14) {materialFunction=&MakeFloat3}.Result
   BasisZ=Constant3Vector() {constant=<rgba 0 0 1 0>}
  ) {materialFunction=&Transform3x3Matrix}
  n16 = LinearInterpolate(
   A=MaterialFunctionCall(Rotation Axis=NamedRerouteUsage(), Rotation Angle=NamedRerouteUsage()) {materialFunction=&FixRotateAboutAxisNormals}.New Tangent Space Vertex Normal
   B=Transform(None=Normalize(VectorInput=@n15.Transformed Vector)) {transformSourceType=TRANSFORMSOURCE_World, transformType=TRANSFORM_Tangent}
   Alpha=Ceil(None=NamedRerouteUsage())
  )
  n17 = LinearInterpolate(A=NamedRerouteUsage(), B=NamedRerouteUsage(), Alpha=Ceil(None=NamedRerouteUsage()))
  n18 = NamedRerouteUsage()
  n19 = Constant()
  n20 = MaterialFunctionCall(
   X=@n19
   Y=@n19
   Z=ScalarParameter() {bUseCustomPrimitiveData=true, defaultValue=300, parameterName=AfterRotateTransformHeight, primitiveDataIndex=1}
  ) {materialFunction=&MakeFloat3}
  n21 = Transform(None=Add(
    A=Multiply(A=@n18, B=@n20.Result)
    B=VectorParameter() {group=LidTransform, parameterName=PivotTransformOffset, defaultValue=<rgba 0 0 0 1>}.RGB
   )) {transformSourceType=TRANSFORMSOURCE_Local}
  FunctionOutput(None=SetMaterialAttributes!(
    MaterialAttributes=@n9.MaterialAttributes
    Normal=MaterialFunctionCall(BaseNormal=@n9.Normal, AdditionalNormal=@n16) {materialFunction=&BlendAngleCorrectedNormals}.Result
    WorldPositionOffset=Add(A=@n17, B=@n21)
   ) {attributeSetTypes [#1, #2]})
  MaterialFunctionCall() {materialFunction=&ObjectPivotPoint}
  Multiply(A=VectorParameter() {group=LidTransform, parameterName=AfterRotateTransformMiddle, defaultValue=<rgba 0 0 300 1>}.RGB)
  n22 = Multiply(A=Time(), B=ScalarParameter() {defaultValue=0.5, group=LidTransform, parameterName=AnimationSpeed}) {constB=0.3}
  n23 = ScalarParameter() {bUseCustomPrimitiveData=true, group=LidTransform, parameterName=FinalProgressTime, sliderMax=1}
  n24 = CurveAtlasRowParameter(CurveTime=StaticSwitchParameter(True=Frac(None=@n22), False=NamedRerouteDeclaration(None=@n23) {name=FinalTime, nodeColor=<rgba 1 0 0.3047 1>}) {defaultValue=true, group=LidTransform, parameterName=PreviewAnimation}) {
   atlas=/SewerPortal/Gameplay/Textures/A_TimeScale_LidOpen_EverGreen
   curve=/SewerPortal/Gameplay/Textures/TimeScaleTest_LidOpen_EverGreen
   group=LidTransform
   parameterName=TimeRemapSpeedVar
  }
  NamedRerouteDeclaration(None=@n24.R) {name=OffsetRotateProgress, nodeColor=<rgba 1 0 0.4219 1>}
  n25 = Multiply(A=Time(), B=ScalarParameter() {defaultValue=0.5, group=LidTransform, parameterName=AnimationSpeed}) {constB=0.3}
  n26 = ScalarParameter() {bUseCustomPrimitiveData=true, group=LidTransform, parameterName=FinalProgressTime, sliderMax=1}
  n27 = CurveAtlasRowParameter(CurveTime=StaticSwitchParameter(True=Frac(None=@n25), False=NamedRerouteDeclaration(None=@n26) {name=FinalTime_1, nodeColor=<rgba 1 0 0.3047 1>}) {defaultValue=true, group=LidTransform, parameterName=PreviewAnimation}) {
   atlas=&A_TimeScale_LidOpen_EverGreen
   curve=&TimeScaleTest_LidOpen_EverGreen
   group=LidTransform
   parameterName=TimeRemapSpeedVar
  }
  NamedRerouteDeclaration(None=@n27.G) {name=HeightProgress_1, nodeColor=<rgba 0 0.9766 1 1>}
  NamedRerouteDeclaration(None=@n27.B) {name=SelfRotateProgress_1, nodeColor=<rgba 1 0.8203 0 1>}
  NamedRerouteDeclaration(None=@n27.A) {name=DampingProgress_1, nodeColor=<rgba 0 0.625 1 1>}
  n28 = Transform(None=NamedRerouteDeclaration(None=VectorParameter() {group=LidTransform, parameterName=LocalRotationAxisFinal, defaultValue=<rgba 1 0 0 1>}.RGB) {name=LocalAxis, nodeColor=<rgba 1 0 0.1406 1>}) {transformSourceType=TRANSFORMSOURCE_Local}
  NamedRerouteDeclaration(None=@n28) {name=WorldAxis, nodeColor=<rgba 0.7812 0 1 1>}
  n29 = Multiply(
   A=NamedRerouteUsage()
   B=ScalarParameter() {defaultValue=-2, desc=This paramter should adjust by hand, group=LidTransform, parameterName=FitRadian}
  )
  NamedRerouteDeclaration(None=RotateAboutAxis(
    NormalizedRotationAxis=@n1
    RotationAngle=@n29
    PivotPoint=MaterialFunctionCall() {materialFunction=&ObjectPivotPoint}.Object Pivot Location
    Position=WorldPosition().XYZ
   )) {name=SelfRotate, nodeColor=<rgba 1 0 0.1406 1>}
  NamedRerouteDeclaration(None=@n24.G) {name=HeightProgress, nodeColor=<rgba 0 0.9766 1 1>}
  NamedRerouteDeclaration(None=@n24.B) {name=SelfRotateProgress, nodeColor=<rgba 1 0.8203 0 1>}
  NamedRerouteDeclaration(None=@n24.A) {name=DampingProgress, nodeColor=<rgba 0 0.625 1 1>}
  NamedRerouteDeclaration(None=@n27.R) {name=OffsetRotateProgress_1, nodeColor=<rgba 1 0 0.4219 1>}
  n30 = Multiply(
   A=NamedRerouteUsage()
   B=ScalarParameter() {defaultValue=-2, desc=This paramter should adjust by hand, group=LidTransform, parameterName=FitRadian}
  )
  NamedRerouteDeclaration(None=RotateAboutAxis(
    NormalizedRotationAxis=@n28
    RotationAngle=NamedRerouteDeclaration(None=@n30) {name=SelfRotateAngle, nodeColor=<rgba 1 0.6094 0 1>}
    PivotPoint=MaterialFunctionCall() {materialFunction=&ObjectPivotPoint}.Object Pivot Location
    Position=WorldPosition().XYZ
   )) {name=SelfRotate_1, nodeColor=<rgba 1 0 0.1406 1>}
  n31 = Multiply(A=NamedRerouteUsage(), B=ScalarParameter() {defaultValue=-0.24, group=LidTransform, parameterName=OffsetRadian}) {constB=-1}
  n32 = VectorParameter() {group=LidTransform, parameterName=XRotatePivot, defaultValue=<rgba 0 50 0 1>}
  n33 = Transform(None=RotateAboutAxis(
    NormalizedRotationAxis=NamedRerouteUsage()
    RotationAngle=NamedRerouteDeclaration(None=@n31) {name=OffsetRotationAngle, nodeColor=<rgba 0.0781 1 0 1>}
    PivotPoint=NamedRerouteDeclaration(None=@n32.RGB) {name=XPivot, nodeColor=<rgba 1 0.0703 0 1>}
    Position=MaterialFunctionCall() {materialFunction=&LocalPosition}.Local Position
   )) {transformSourceType=TRANSFORMSOURCE_Local}
  n34 = WorldPosition()
  n35 = NamedRerouteUsage()
  n36 = Cosine(None=@n35)
  n37 = Sine(None=@n35)
  n38 = Constant()
  n39 = MaterialFunctionCall(
   VectorToTransform=TransformPosition(None=Add(A=@n33, B=@n34.XYZ)) {transformSourceType=TRANSFORMPOSSOURCE_World}
   BasisX=MaterialFunctionCall(X=@n36, Y=Multiply(A=@n37) {constB=-1}, Z=@n38) {materialFunction=&MakeFloat3}.Result
   BasisY=MaterialFunctionCall(X=@n37, Y=@n36, Z=@n38) {materialFunction=&MakeFloat3}.Result
   BasisZ=Constant3Vector() {constant=<rgba 0 0 1 0>}
   BasisPosition=Constant3Vector()
  ) {materialFunction=&Transform3x3Matrix}
  NamedRerouteDeclaration(None=Subtract(A=TransformPosition(None=@n39.Transformed Vector) {transformType=TRANSFORMPOSSOURCE_World}, B=WorldPosition().XYZ)) {name=OffsetRotate_1, nodeColor=<rgba 0 1 0.4609 1>}
  MaterialExpressionReroute_11 = Reroute(None=@n7)
  MaterialExpressionReroute_13 = Reroute(None=@n8)
  n40 = Cosine(None=@MaterialExpressionReroute_13.Input2)
  MaterialExpressionReroute_14 = Reroute(None=@n8)
  n41 = Sine(None=@MaterialExpressionReroute_14.Input3)
  n42 = Multiply(A=@n41) {constB=-1}
  n43 = Constant()
  MaterialExpressionReroute_12 = Reroute(None=@n5.Object Pivot Location)
  n44 = Transform(None=MaterialFunctionCall(
    VectorToTransform=TransformPosition(None=@MaterialExpressionReroute_11.Input) {transformSourceType=TRANSFORMPOSSOURCE_World}
    BasisX=MaterialFunctionCall(X=@n40, Y=@n42, Z=@n43) {materialFunction=&MakeFloat3}.Result
    BasisY=MaterialFunctionCall(X=@n41, Y=@n40, Z=@n43) {materialFunction=&MakeFloat3}.Result
    BasisZ=Constant3Vector() {constant=<rgba 0 0 1 0>}
    BasisPosition=@MaterialExpressionReroute_12."BasisPosition (V3)"
   ) {materialFunction=&Transform3x3Matrix}.Transformed Vector) {transformSourceType=TRANSFORMSOURCE_Local}
  PinBase!(Output=@n44) {reroutePins [Output {expression=&self:MaterialExpressionReroute_15}]}
  PinBase!() {reroutePins [
    Input {expression=&self:MaterialExpressionReroute_11}
    "BasisPosition (V3)" {expression=&self:MaterialExpressionReroute_12}
    Input2 {expression=&self:MaterialExpressionReroute_13}
    Input3 {expression=&self:MaterialExpressionReroute_14}
   ]}
  n45 = GetMaterialAttributes() {attributeGetTypes [#1]}
  n46 = MaterialFunctionCall(Rotation Axis=@n1, Rotation Angle=@n29) {materialFunction=&FixRotateAboutAxisNormals}
  n47 = Multiply(A=NamedRerouteUsage(), B=@n3) {constB=-1}
  n48 = Normalize(VectorInput=Transform(None=MaterialFunctionCall(Rotation Axis=@n1, Rotation Angle=@n47) {materialFunction=&FixRotateAboutAxisNormals}.New Tangent Space Vertex Normal))
  n49 = NamedRerouteUsage()
  n50 = Cosine(None=@n49)
  n51 = Sine(None=@n49)
  n52 = Constant()
  n53 = MaterialFunctionCall(X=@n50, Y=Multiply(A=@n51) {constB=-1}, Z=@n52) {materialFunction=&MakeFloat3}
  n54 = MaterialFunctionCall(X=@n51, Y=@n50, Z=@n52) {materialFunction=&MakeFloat3}
  n55 = Constant3Vector() {constant=<rgba 0 0 1 0>}
  n56 = Transform(None=Normalize(VectorInput=MaterialFunctionCall(VectorToTransform=@n48, BasisX=@n53.Result, BasisY=@n54.Result, BasisZ=@n55) {materialFunction=&Transform3x3Matrix}.Transformed Vector)) {transformSourceType=TRANSFORMSOURCE_World, transformType=TRANSFORM_Tangent}
  n57 = Ceil(None=NamedRerouteUsage())
  n58 = NamedRerouteUsage()
  MaterialExpressionReroute_15 = Reroute(None=@n44)
  n59 = NamedRerouteDeclaration(None=Subtract(A=@MaterialExpressionReroute_15.Output, B=@n6.XYZ)) {name=OffsetRotate, nodeColor=<rgba 0 1 0.4609 1>}
  n60 = Ceil(None=NamedRerouteUsage())
  n61 = Constant()
  n62 = Multiply(
   A=NamedRerouteUsage()
   B=MaterialFunctionCall(
    X=@n61
    Y=@n61
    Z=ScalarParameter() {bUseCustomPrimitiveData=true, defaultValue=300, parameterName=AfterRotateTransformHeight, primitiveDataIndex=1}
   ) {materialFunction=&MakeFloat3}.Result
  )
  n63 = VectorParameter() {group=LidTransform, parameterName=PivotTransformOffset, defaultValue=<rgba 0 0 0 1>}
  SetMaterialAttributes(
   MaterialAttributes=@n45.MaterialAttributes
   Normal=MaterialFunctionCall(BaseNormal=@n45.Normal, AdditionalNormal=LinearInterpolate(A=@n46.New Tangent Space Vertex Normal, B=@n56, Alpha=@n57)) {materialFunction=&BlendAngleCorrectedNormals}.Result
   WorldPositionOffset=Add(A=LinearInterpolate(A=@n58, B=@n59, Alpha=@n60), B=Add(A=@n62, B=@n63.RGB))
  ) {attributeSetTypes [#1, #2]}
 ]}
