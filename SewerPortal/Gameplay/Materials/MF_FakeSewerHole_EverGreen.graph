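# Material function that fakes a deep brick-lined sewer hole, blended into incoming material attributes by BlendAlpha.
# A ray-cylinder intersection at the actor position yields interior wall points, textured by two axis-projected brick samples chosen by position.
# Depth darkens the bricks, time of day recolors them, and the GreenEmission switch adds a tinted glow strengthening with depth.
# The result is emissive, divided by eye adaptation, over zeroed surface inputs.
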MaterialFunction /SewerPortal/Gameplay/Materials/MF_FakeSewerHole_EverGreen
graph {nodes [
  n1 = ScalarParameter() {defaultValue=100, group=Cylinder, parameterName=ObjecRadiusRelative}
  n2 = VectorParameter() {group=Cylinder, parameterName=ZMin-Max, channelNames {a="", b="", g=ZMax, r=ZMin}, defaultValue=<rgba -600 0 0 1>}
  n3 = Subtract(
   A=MaterialFunctionCall!(
    Origin=ActorPositionWS()
    Radius=NamedRerouteDeclaration(None=@n1) {name=Radius, nodeColor=<rgba 0.5469 0 1 1>}
    Z Axis=Constant3Vector() {constant=<rgba 0 0 1 0>}
    XY Scale
    X Axis
    Capped?=StaticBool() {value=true}
    Z Min-Max=@n2
   ) {materialFunction=/Engine/Functions/Engine_MaterialFunctions01/Math/CylinderIntersection}.Ray Exit Position
   B=ActorPositionWS()
  )
  n4 = Abs(None=ComponentMask(None=@n3) {b=true, r=true})
  n5 = Add(
   A=NamedRerouteDeclaration(None=Constant2Vector() {r=0.5}) {name=UVOffset, nodeColor=<rgba 0.3125 0 1 1>}
   B=Divide(A=@n4, B=NamedRerouteUsage()) {constB=100}
  )
  n6 = ComponentMask(None=@n3) {b=true, g=true}
  n7 = Add(A=Divide(A=Abs(None=@n6), B=NamedRerouteUsage()) {constB=100}, B=NamedRerouteUsage())
  n8 = ComponentMask(None=Divide(A=@n4, B=NamedRerouteUsage())) {r=true}
  n9 = Divide(A=Constant() {r=1}, B=SquareRoot(None=Constant() {r=2}))
  n10 = Constant() {r=1}
  n11 = Constant()
  n12 = LinearInterpolate(
   A=TextureSample(
    UVs=NamedRerouteDeclaration(None=@n5) {name=UV1, nodeColor=<rgba 1 0.8906 0 1>}
    Tex=TextureObjectParameter() {group=Cylinder, parameterName=Bricks01, texture=/SewerPortal/Gameplay/Textures/T_TunnelBrickColor_Vertical_01_EverGreen}
   ) {texture=/Game/FN_PublicTurntable/Turntable_Lowpoly/Content/BlueprintAssets/Checker}.RGB
   B=TextureSample(
    UVs=NamedRerouteDeclaration(None=@n7) {name=UV2, nodeColor=<rgba 0.5703 0 1 1>}
    Tex=TextureObjectParameter() {group=Cylinder, parameterName=Bricks02, texture=&T_TunnelBrickColor_Vertical_01_EverGreen}
   ) {texture=&Checker}.RGB
   Alpha=NamedRerouteDeclaration(None=If(A=@n8, B=@n9, "A > B"=@n10, "A == B"=@n11, "A < B"=@n11) {constB=0.5}) {name=Alpha, nodeColor=<rgba 0.2656 1 0 1>}
  )
  n13 = ComponentMask(None=@n3) {b=true}
  n14 = ScalarParameter() {defaultValue=300, group=Cylinder, parameterName=DepthShadow}
  n15 = Saturate(None=Divide(A=Add(A=@n13, B=@n14) {constB=100}, B=@n14) {constB=5})
  n16 = LinearInterpolate(Alpha=@n15) {constA=0.1, constB=0.7}
  n17 = OneMinus(None=@n15)
  n18 = NamedRerouteDeclaration(None=Multiply(A=Multiply(A=@n12, B=@n16), B=LinearInterpolate(Alpha=@n17) {constA=0.6})) {name=HoleColor, nodeColor=<rgba 1 0.8203 0 1>}
  n19 = Constant3Vector() {constant=<rgba 0.1119 0.2246 0.6146 1>}
  n20 = ScalarParameter() {defaultValue=700, group=Cylinder, parameterName=EmissiveGlow}
  n21 = Add(A=@n13, B=@n20)
  n22 = Power(Base=Saturate(None=Divide(A=@n21, B=@n20))) {constExponent=0.7}
  n23 = Multiply(
   A=VectorParameter() {group=Cylinder, parameterName=HoleColorTint, defaultValue=<rgba 0 1 0 1>}.RGB
   B=Multiply(A=NamedRerouteUsage(), B=OneMinus(None=@n22))
  )
  n24 = Divide(A=@n23) {constB=1.5}
  n25 = Add(
   A=MaterialFunctionCall(Color_Morning=@n18, Color_Day=@n18, Color_Evening=Divide(A=@n18) {constB=2.5}, Color_Night=Multiply(A=@n18, B=@n19)) {materialFunction=&Color_Time_of_Day}.Result
   B=StaticSwitchParameter(
    True=MaterialFunctionCall(Color_Morning=@n23, Color_Day=@n23, Color_Evening=@n24, Color_Night=@n24) {materialFunction=&Color_Time_of_Day}.Result
    False=Constant()
   ) {group=EverGreen, parameterName=GreenEmission}
  )
  n26 = ScalarParameter() {defaultValue=1, group=Cylinder, parameterName=EmissiveIntensity}
  n27 = Constant()
  n28 = SetMaterialAttributes!(
   MaterialAttributes
   EmissiveColor=Divide(A=Multiply(A=@n25, B=@n26), B=EyeAdaptation().EyeAdaptation)
   ShadingModel=ShadingModel()
   BaseColor=@n27
   Metallic=@n27
   Specular=@n27
   Roughness=@n27
  ) {attributeSetTypes [
    #1
    #2
    #3
    #4
    #5
    #6
   ]}
  FunctionOutput(None=BlendMaterialAttributes(
    A=FunctionInput() {inputName=MA, inputType=FunctionInput_MaterialAttributes}
    B=@n28
    Alpha=FunctionInput() {inputName=BlendAlpha, inputType=FunctionInput_Scalar, sortPriority=1}
   ))
  NamedRerouteDeclaration(None=Divide(A=NamedRerouteUsage(), B=ScalarParameter() {defaultValue=0.5, group=Cylinder, parameterName=Tiling})) {name=RadiusBasedTilling, nodeColor=<rgba 0 0.9297 1 1>}
  NamedRerouteDeclaration(None=Divide(A=ComponentMask(None=@n3) {g=true, r=true}, B=NamedRerouteUsage()) {constB=100}) {name=UV3, nodeColor=<rgba 0.2891 0 1 1>}
  n29 = TextureSampleParameter2D(UVs=Rotator(Coordinate=NamedRerouteUsage()) {centerX=0, centerY=0}) {parameterName=NoiseEM, samplerType=SAMPLERTYPE_Masks, texture=/SewerPortal/Gameplay/Textures/T_BrickColor_Vertical_Mask_EverGreen}
  NamedRerouteDeclaration(None=LinearInterpolate(Alpha=@n29.G) {constA=0.6}) {name=NoiseEM, nodeColor=<rgba 0 0.2031 1 1>}
 ]}
