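# Fortnite master material for the SewerPortal fake sewer hole, an EverGreen variant built on the 2022/2023 layered base material (GigaMix).
# The base layers pass through shared damage, wind, fan-spin and fire-fallback functions, convert normals to world space, then gain DirectFX, seasonal jungle biome and RVT blending.
# A sewer-hole function composites the tunnel effect using the FX_Map red channel as blend alpha.
# Mobile swaps in constant specular and roughness before the Creative move-tool function.
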Material /SewerPortal/Gameplay/Materials/M_FN_FakeSewerHole_EverGreen
graph {
 properties {
  bAutomaticallySetUsageInEditor=false
  bTangentSpaceNormal=false
  bUseMaterialAttributes=true
  bUsedWithInstancedStaticMeshes=true
  bUsedWithNanite=true
  ditheredLODTransition=true
  pixelDepthOffsetMode=PDOM_Legacy
  shadingModel=MSM_FromMaterialExpression
  usedShadingModels=Default Lit | Subsurface | Clear Coat | Two Sided Foliage
 }
 nodes [
  n1 = MaterialFunctionCall() {materialFunction=/Game/Global/Base4Layers_2022/MaterialFunctions/MF_GigaMix_2023}
  MaterialExpressionReroute_70 = Reroute(None=@n1.Result)
  n2 = ScalarParameter() {
   desc=Documentation link
   group="0 (Expand)  Documentation (RMB - Copy display name) and Debug view"
   parameterName=https://confluence.it.epicgames.com/display/FORT/Fortnite+Base+Material+2022
  }
  n3 = SetMaterialAttributes!(MaterialAttributes=@MaterialExpressionReroute_70.MaterialAttributes, World Position Offset=Multiply(A=@n2)) {attributeSetTypes [#1]}
  MaterialExpressionReroute_71 = Reroute(None=@n3)
  MaterialExpressionReroute_69 = Reroute(None=@MaterialExpressionReroute_71.Output)
  n4 = MaterialFunctionCall(Material_In=@MaterialExpressionReroute_69.MaterialAttributes) {materialFunction=&MF_FN_DestructibleDamage}
  n5 = MaterialFunctionCall!(
   In=MaterialFunctionCall!(Material Attributes=MaterialFunctionCall!(In=@n4.Material_Out) {materialFunction=/Game/Athena/WinterArt/Materials/MF_WPO_WindNew}.Result) {materialFunction=/Game/Creative/Sets/MilitaryBase/Props/MF_FanSpin}.Result
   UseSphericalMappingOverride
  ) {materialFunction=/Game/Packages/Fortress_SharedMaterials/Base_Material_Functions/MF_CurieFallbackEffect}
  MaterialExpressionReroute_68 = Reroute(None=@n5.Result)
  Composite!(None=@MaterialExpressionReroute_68.Result) {desc=Transform to World in case we will use RVT after this point, subgraphName=Tangent to World}
  Composite!(MaterialAttributes=@n1.Result) {subgraphName=Documentation Link}
  Composite(MaterialAttributes=@MaterialExpressionReroute_71.Output) {desc=Game Shared WPO and Fire Effect, subgraphName=WPO And Shared Functions}
  MaterialExpressionReroute_1 = Reroute(None=@n10.RVT H)
  MaterialExpressionReroute_3 = Reroute(None=@MaterialExpressionReroute_68.Result)
  n6 = GetMaterialAttributes(None=@MaterialExpressionReroute_3.Input) {attributeGetTypes [#2]}
  n7 = SetMaterialAttributes!(MaterialAttributes=@n6.MaterialAttributes, Normal=Transform(None=@n6.Normal)) {attributeSetTypes [#2]}
  MaterialExpressionReroute_2 = Reroute(None=@n7)
  n8 = MaterialFunctionCall!(In=@MaterialExpressionReroute_2.Output) {materialFunction=/Game/ArtTools/Tech_Art/DirectFX/Assets/MaterialFunctions/MF_DirectFX}
  MaterialExpressionReroute_8 = Reroute(None=@n8.Result)
  n9 = MaterialFunctionCall!(AO, RVT_Height=@MaterialExpressionReroute_1."RVT_Height (S)", For Buildings=@MaterialExpressionReroute_8."For Moss (MA)", For Moss) {materialFunction=/Game/Environments/Asteria/Rocks/Materials/MaterialFunctions/MF_JungleBiome}
  MaterialExpressionReroute_0 = Reroute(None=@n9.Buildings Jungl)
  n10 = MaterialFunctionCall!(
   In=@MaterialExpressionReroute_0.Result
   "RVT Height Map (Roughness)"=GetMaterialAttributes(None=@n8.Result) {attributeGetTypes [#3]}.Roughness
  ) {materialFunction=/Game/Global/Base4Layers_2022/MaterialFunctions/MF_RVTBlending}
  n11 = MaterialFunctionCall!(
   MA=@n10.Result
   BlendAlpha=TextureSampleParameter2D() {group=Cylinder, parameterName=FX_Map, texture=/SewerPortal/Gameplay/Textures/T_Tunnel_Sewer_FX}.R
  ) {materialFunction=&/MF_FakeSewerHole_EverGreen}
  Composite!(InputPin=@n11.Result) {subgraphName=Mobile Opt}
  Composite!("For Moss (MA)"=@n8.Result, "RVT_Height (S)"=@n10.RVT H) {subgraphName="Seasonal "}
  MaterialExpressionReroute_7 = Reroute(None=@n11.Result)
  n12 = GetMaterialAttributes(None=@MaterialExpressionReroute_7.InputPin) {attributeGetTypes [#4, #2]}
  n13 = SetMaterialAttributes!(
   MaterialAttributes=@n12.MaterialAttributes
   Specular=Constant() {r=0.5}
   Roughness=Constant() {r=0.8}
   Base Color=@n12.MaterialAttributes
   Normal=@n12.MaterialAttributes
  ) {attributeSetTypes [
    #5
    #3
    #4
    #2
   ]}
  n14 = FeatureLevelSwitch(Default=@MaterialExpressionReroute_7.InputPin, ES3_1=@n13)
  MaterialExpressionReroute_6 = Reroute(None=@n14)
  MaterialFunctionCall(MA In=@MaterialExpressionReroute_6.Output) {materialFunction=&MF_CreativeMoveTool}
  PinBase() {reroutePins [Input {expression=&self:MaterialExpressionReroute_3}]}
  PinBase!(Output=@n7) {reroutePins [Output {expression=&self:MaterialExpressionReroute_2}]}
  PinBase!() {reroutePins [
    "For Moss (MA)" {expression=&self:MaterialExpressionReroute_8}
    "RVT_Height (S)" {expression=&self:MaterialExpressionReroute_1}
   ]}
  PinBase!(Result=@n9.Buildings Jungl) {reroutePins [Result {expression=&self:MaterialExpressionReroute_0}]}
  PinBase!() {reroutePins [MaterialAttributes {expression=&self:MaterialExpressionReroute_70}]}
  PinBase!(Output=@n3) {reroutePins [Output {expression=&self:MaterialExpressionReroute_71}]}
  PinBase!() {reroutePins [MaterialAttributes {expression=&self:MaterialExpressionReroute_69}]}
  PinBase!(Result=@n5.Result) {reroutePins [Result {expression=&self:MaterialExpressionReroute_68}]}
  PinBase!() {reroutePins [InputPin {expression=&self:MaterialExpressionReroute_7}]}
  PinBase!(Output=@n14) {reroutePins [Output {expression=&self:MaterialExpressionReroute_6}]}
 ]
}
