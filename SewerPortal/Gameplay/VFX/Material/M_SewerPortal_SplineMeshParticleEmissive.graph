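# Unlit, two-sided translucent material for Niagara mesh particles that draws glowing spline-thickened lines in the Sewer Portal VFX.
# SplineThicken takes one Width parameter for both base and tip, while particle color RGB runs through the shared FX emissive adjustment function.
# Particle alpha is multiplied by a soft sphere mask centered along U and raised to a Fade exponent, fading both ends.
# The WidthTip parameter and a SmoothStep node are unconnected.
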Material /SewerPortal/Gameplay/VFX/Material/M_SewerPortal_SplineMeshParticleEmissive
graph {
 properties {
  bUsedWithNiagaraMeshParticles=true
  blendMode=BLEND_Translucent
  pixelDepthOffsetMode=PDOM_Legacy
  previewMesh=/Game/Animation/Game/MainPlayer/Emotes/StarStray/FX/SM_Emote_StarStray_SplineThickenCircle
  shadingModel=MSM_Unlit
  twoSided=true
  usedShadingModels=Unlit
 }
 nodes [
  n1 = ScalarParameter() {defaultValue=5, parameterName=Width, sliderMax=10}
  MaterialFunctionCall(WidthBase=@n1, WidthTip=@n1) {materialFunction=&SplineThicken}
  n2 = ParticleColor()
  MaterialFunctionCall(Emissive=@n2.RGB) {materialFunction=&MF_FX_EmissiveAdjustments}
  n3 = ComponentMask(None=TextureCoordinate()) {r=true}
  n4 = Constant() {r=0.5}
  Multiply(
   A=@n2.A
   B=Power(
    Base=SphereMask(A=@n3, B=@n4) {attenuationRadius=0.5, hardnessPercent=0}
    Exp=ScalarParameter() {defaultValue=1, parameterName=Fade, sliderMax=10}
   ) {constExponent=4}
  )
  ScalarParameter() {defaultValue=5, parameterName=WidthTip, sliderMax=10}
  SmoothStep()
 ]
}
