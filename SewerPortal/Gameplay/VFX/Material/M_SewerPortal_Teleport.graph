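# Unlit translucent Niagara mesh-particle material for the sewer portal teleport, drawing a scrolling brick tunnel on a screen cone without depth testing.
# Panned bricks combine offset, mip-biased blur samples (single-sample on Low/mobile), normal-driven fake light and distorted cubemap tint.
# Graffiti, pulsing speed lines and gradient shadows add detail, while sine/cosine world offset wobbles vertices.
# Dynamic parameters drive exit white-wash and emissive strength, with a static switch to debug scalars.
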Material /SewerPortal/Gameplay/VFX/Material/M_SewerPortal_Teleport
graph {
 properties {
  bDisableDepthTest=true
  bUseTranslucencyVertexFog=false
  bUsedWithNiagaraMeshParticles=true
  blendMode=BLEND_Translucent
  pixelDepthOffsetMode=PDOM_Legacy
  previewMesh=/SewerPortal/Gameplay/VFX/SM_SewerPortal_ScreenCone
  shadingModel=MSM_Unlit
  translucencyPass=MTP_AfterMotionBlur
  usedShadingModels=Unlit
 }
 nodes [
  n1 = TextureObjectParameter() {
   group=TunnelTransition
   parameterName=01_BricksNormal
   samplerType=SAMPLERTYPE_Normal
   texture=/SewerPortal/Gameplay/VFX/Texture/T_Bricks_Cracks_V3_N
  }
  n2 = ScalarParameter() {defaultValue=4, group=UVs, parameterName=UV U Mult}
  n3 = ScalarParameter() {defaultValue=4, group=UVs, parameterName=UV V Mult}
  n4 = ScalarParameter() {defaultValue=0.6, group=UVs, parameterName=UV U Pan Speed}
  n5 = ScalarParameter() {defaultValue=-3, group=UVs, parameterName=UV V Pan Speed}
  n6 = Rotator(
   Coordinate=MaterialFunctionCall!(
    UV Channel=Constant()
    U Freq Multiplier=NamedRerouteDeclaration(None=@n2) {name=UMulti01, nodeColor=<rgba 0.5 1 0 1>}
    V Freq Multiplier=NamedRerouteDeclaration(None=@n3) {name=VMulti01, nodeColor=<rgba 1 0 0.8437 1>}
    Pan Speed U=NamedRerouteDeclaration(None=@n4) {name=USpeed01, nodeColor=<rgba 0 1 0.4844 1>}
    Pan Speed V=NamedRerouteDeclaration(None=@n5) {name=Vspeed01, nodeColor=<rgba 0 0.5078 1 1>}
   ) {materialFunction=/Game/Effects/Niagara/MasterMaterials/Functions/UVs/MF_FX_TextureUVPanner}.Result
   Time=Constant() {r=4.7125}
  ) {speed=1}
  n7 = TextureSample(UVs=@n6, Tex=@n1) {samplerSource=SSM_Wrap_WorldGroupSettings, samplerType=SAMPLERTYPE_Normal}
  n8 = StaticBool()
  n9 = TextureSample!(UVs=Add(A=@n6, B=Constant2Vector() {r=-0.0025}), Tex=@n1, Bias=Constant() {r=1}, Apply View MipBias=@n8) {mipValueMode=TMVM_MipBias, samplerSource=SSM_Wrap_WorldGroupSettings, samplerType=SAMPLERTYPE_Normal}
  n10 = Add(A=@n6, B=Constant2Vector() {r=-0.005})
  n11 = Constant() {r=2}
  n12 = Constant2Vector() {r=-0.0075}
  n13 = Add(
   A=Add(
    A=Add(A=@n7.RGB, B=@n9.RGB)
    B=TextureSample(UVs=@n10, Tex=@n1, Bias=@n11, Apply View MipBias=@n8) {mipValueMode=TMVM_MipBias, samplerSource=SSM_Wrap_WorldGroupSettings, samplerType=SAMPLERTYPE_Normal}.RGB
   )
   B=TextureSample(UVs=Add(A=@n6, B=@n12), Tex=@n1, Bias=Constant() {r=2}, Apply View MipBias=@n8) {mipValueMode=TMVM_MipBias, samplerSource=SSM_Wrap_WorldGroupSettings, samplerType=SAMPLERTYPE_Normal}.RGB
  )
  n14 = QualitySwitch(Default=ShadingPathSwitch(Default=Divide(A=@n13) {constB=4}, Mobile=@n7.RGB) {desc=Migrated from FeatureLevelSwitch}, Low=@n7.RGB)
  n15 = TextureSample!(UVs=@n6, Tex, Apply View MipBias) {samplerSource=SSM_Wrap_WorldGroupSettings, samplerType=SAMPLERTYPE_Normal, texture=/Game/Global/Textures/Default/Tileable_Patterns/1Normal}
  n16 = MaterialFunctionCall(Normal=@n15.RGB, Flatness=ScalarParameter() {defaultValue=0.925, group=Cubemap, parameterName=Cubemap Distortion Flatness}) {materialFunction=&FlattenNormal}
  n17 = QualitySwitch(
   Default=ShadingPathSwitch(
    Default=MaterialFunctionCall(BaseNormal=@n14, AdditionalNormal=@n16.Result) {materialFunction=&BlendAngleCorrectedNormals}.Result
    Mobile=@n14
   ) {desc=Migrated from FeatureLevelSwitch}
   Low=@n14
  )
  n18 = DotProduct(
   A=Transform(None=VectorParameter() {group=TunnelTransition, parameterName=FakeLightPos, defaultValue=<rgba 0 0 -0.05 1>}.RGB) {transformSourceType=TRANSFORMSOURCE_World, transformType=TRANSFORM_Tangent}
   B=@n17
  )
  LinearInterpolate(
   A=VectorParameter() {group=TunnelTransition, parameterName=ColorBright, defaultValue=<rgba 0.2231 0.3229 0.2686 1>}.RGB
   B=VectorParameter() {group=TunnelTransition, parameterName=ColorDark, defaultValue=<rgba 60 50 100 1>}.RGB
   Alpha=NamedRerouteDeclaration(None=Clamp(None=@n18)) {name=FakeLightMap, nodeColor=<rgba 1 0.5391 0 1>}
  )
  n19 = TextureObjectParameter() {
   group=TunnelTransition
   parameterName=01_BrickNoise
   samplerType=SAMPLERTYPE_Masks
   texture=/SewerPortal/Gameplay/VFX/Texture/T_BrickColor_Mask
  }
  n20 = NamedRerouteDeclaration(None=@n6) {name=BaseUVCoord, nodeColor=<rgba 1 0 0.8437 1>}
  n21 = TextureSample(UVs=@n20, Tex=@n19) {samplerType=SAMPLERTYPE_Masks}
  NamedRerouteDeclaration(None=Multiply(A=OneMinus(None=@n21.R), B=@n21.G)) {name=CloudNoise, nodeColor=<rgba 0 0.6484 1 1>}
  n22 = TextureObjectParameter() {group=TunnelTransition, parameterName=01_Diffuse01, texture=/SewerPortal/Gameplay/VFX/Texture/T_Bricks_Cracks_D_01}
  n23 = TextureSample(UVs=@n20, Tex=@n22) {samplerSource=SSM_Wrap_WorldGroupSettings}
  n24 = StaticBool()
  n25 = TextureSample!(UVs=Add(A=@n20, B=Constant2Vector() {r=-0.0025}), Tex=@n22, Bias=Constant() {r=1}, Apply View MipBias=@n24) {mipValueMode=TMVM_MipBias, samplerSource=SSM_Wrap_WorldGroupSettings}
  n26 = Add(A=@n20, B=Constant2Vector() {r=-0.005})
  n27 = Constant() {r=2}
  n28 = Constant2Vector() {r=-0.0075}
  n29 = Add(
   A=Add(
    A=Add(A=@n23.RGB, B=@n25.RGB)
    B=TextureSample(UVs=@n26, Tex=@n22, Bias=@n27, Apply View MipBias=@n24) {mipValueMode=TMVM_MipBias, samplerSource=SSM_Wrap_WorldGroupSettings}.RGB
   )
   B=TextureSample(UVs=Add(A=@n20, B=@n28), Tex=@n22, Bias=Constant() {r=2}, Apply View MipBias=@n24) {mipValueMode=TMVM_MipBias, samplerSource=SSM_Wrap_WorldGroupSettings}.RGB
  )
  n30 = QualitySwitch(Default=ShadingPathSwitch(Default=Divide(A=@n29) {constB=4}, Mobile=@n23.RGB) {desc=Migrated from FeatureLevelSwitch}, Low=@n23.RGB)
  n31 = NamedRerouteUsage()
  n32 = NamedRerouteUsage()
  n33 = TextureSampleParameterCube!(UVs=Transform(None=@n17), Apply View MipBias) {
   group=Cubemap
   parameterName=Cubemap Texture
   samplerSource=SSM_Wrap_WorldGroupSettings
   texture=/Game/Packages/Fortress_Sky/SkyDome/MaterialInstances/SkyDome_Apollo/Cubemap_Apollo_Night
  }
  n34 = ScalarParameter() {defaultValue=1, group=Cubemap, parameterName=Cubemap Desaturation}
  n35 = Add(
   A=Add(A=LinearInterpolate(A=@n30, B=@n31, Alpha=@n32), B=NamedRerouteUsage())
   B=Multiply(
    A=Desaturation(None=@n33.RGB, Fraction=@n34)
    B=VectorParameter() {group=Cubemap, parameterName=Cubemap Multiplier, defaultValue=<rgba 0.1728 0.4688 0.2832 1>}.RGB
   )
  )
  NamedRerouteDeclaration(None=QualitySwitch(Default=ShadingPathSwitch(Default=@n35, Mobile=@n35) {desc=Migrated from FeatureLevelSwitch}, Low=@n35)) {name=TunnelColor, nodeColor=<rgba 0 1 0.7891 1>}
  n36 = VectorParameter() {group=TunnelTransition, parameterName=GhostGreen, defaultValue=<rgba 0.1124 0.7708 0 1>}
  n37 = Multiply(A=@n36.RGB, B=NamedRerouteUsage()) {constB=0.1}
  n38 = Multiply(
   A=Multiply(A=Constant() {r=1}, B=ScalarParameter() {defaultValue=0.4, group=TunnelTransition, parameterName=OverAllBrightness})
   B=Add(A=NamedRerouteUsage(), B=NamedRerouteUsage())
  )
  n39 = MaterialFunctionCall() {materialFunction=&LinearGradient}
  n40 = ScalarParameter() {defaultValue=3, group=TunnelTransition, parameterName=GradientPowerInnerShadow}
  n41 = LinearInterpolate(
   A=VectorParameter() {group=TunnelTransition, parameterName=Tunnel Mist Color, defaultValue=<rgba 0.0338 0.3333 0.033 1>}.RGB
   Alpha=NamedRerouteDeclaration(None=Power(Base=@n39.VGradient, Exp=@n40)) {name=InnerShadow, nodeColor=<rgba 1 0.375 0 1>}
  )
  n42 = Power(Base=@n39.VGradient, Exp=ScalarParameter() {defaultValue=4, group=TunnelTransition, parameterName=GradientPowerInnerShadowDark})
  n43 = Multiply(
   A=Multiply(A=Multiply(A=@n38, B=@n41), B=LinearInterpolate(Alpha=@n42) {constA=0.2})
   B=Power(Base=OneMinus(None=@n39.VGradient), Exp=ScalarParameter() {defaultValue=0.2, group=TunnelTransition, parameterName=GradientPowerOuterShadow})
  )
  n44 = MaterialFunctionCall() {materialFunction=&LinearGradient}
  n45 = SmoothStep(Value=@n44.VGradient) {constMax=0.6, constMin=0.3}
  NamedRerouteDeclaration(None=LinearInterpolate(A=@n36.RGB, B=LinearInterpolate(A=@n37, B=@n43, Alpha=@n45), Alpha=SmoothStep(Value=@n44.VGradient) {constMax=0.4, constMin=0.2})) {name=FinalEM, nodeColor=<rgba 0.9219 1 0 1>}
  n46 = Multiply(A=Time(), B=ScalarParameter() {defaultValue=0.3, group=TunnelTransition, parameterName=RotationSpeed})
  n47 = Add(A=@n46, B=Sine(None=Multiply(A=@n46) {constB=0.45}))
  n48 = Multiply(
   A=MaterialFunctionCall(R=Sine(None=@n47), B=Cosine(None=@n47)) {materialFunction=&AppendMany}.RGB
   B=ScalarParameter() {defaultValue=150, group=TunnelTransition, parameterName=Multiplier}
  )
  n49 = Distance(
   A=WorldPosition() {worldPositionShaderOffset=WPT_ExcludeAllShaderOffsets}.XYZ
   B=TransformPosition(None=VectorParameter() {group=TunnelTransition, parameterName=Offset, defaultValue=<rgba 0 50 0 1>}.RGB) {transformType=TRANSFORMPOSSOURCE_World}
  )
  n50 = ScalarParameter() {defaultValue=150, group=TunnelTransition, parameterName=Rad}
  n51 = OneMinus(None=Saturate(None=Divide(A=@n49, B=@n50)))
  n52 = ScalarParameter() {defaultValue=2, group=TunnelTransition, parameterName=LinearFalloffPower}
  NamedRerouteDeclaration(None=LinearInterpolate(A=Constant3Vector(), B=Transform(None=@n48) {transformSourceType=TRANSFORMSOURCE_Local}, Alpha=Power(Base=@n51, Exp=@n52))) {name=WPO, nodeColor=<rgba 0.5 1 0 1>}
  n53 = NamedRerouteUsage()
  n54 = Subtract(A=Floor(None=ComponentMask(None=@n53) {g=true}), B=Multiply(A=NamedRerouteUsage(), B=Time()))
  n55 = Fmod(A=Floor(None=@n54), B=NamedRerouteUsage())
  n56 = Constant()
  n57 = Constant() {r=1}
  n58 = Add(A=@n54) {constB=2}
  n59 = Constant()
  n60 = If(A=Fmod(A=Floor(None=@n58), B=NamedRerouteUsage()), "A > B"=@n59, "A == B"=Constant() {r=1}, "A < B"=@n59)
  NamedRerouteDeclaration(None=Add(A=If(A=@n55, "A > B"=@n56, "A == B"=@n57, "A < B"=@n56), B=@n60)) {name=D2Alpha, nodeColor=<rgba 1 0.7969 0 1>}
  NamedRerouteDeclaration(None=TextureSample!(
    UVs=@n20
    Tex=TextureObjectParameter() {group=TunnelTransition, parameterName=01_Diffuse02, texture=/SewerPortal/Gameplay/VFX/Texture/T_Bricks_Cracks_stair_D_02}
    Apply View MipBias
   ).RGB) {name=Diffuse02, nodeColor=<rgba 0 1 0.9766 1>}
  NamedRerouteDeclaration(None=Multiply(A=@n60, B=NamedRerouteUsage())) {name=WeaponAlpha, nodeColor=<rgba 1 0.7969 0 1>}
  n61 = ComponentMask(None=NamedRerouteUsage()) {r=true}
  n62 = Multiply(A=NamedRerouteUsage(), B=Time())
  n63 = Floor(None=Subtract(A=Floor(None=@n61), B=Abs(None=@n62)))
  n64 = NamedRerouteUsage()
  n65 = Constant()
  NamedRerouteDeclaration(None=If(A=Fmod(A=@n63, B=@n64), "A > B"=@n65, "A == B"=Constant() {r=1}, "A < B"=@n65)) {name=WeaponAlphaVertical, nodeColor=<rgba 0.6172 1 0 1>}
  n66 = Power(Base=@n21.G)
  n67 = Sine(None=Multiply(A=Time()) {constB=0.5})
  n68 = Multiply(
   A=VectorParameter() {group=TunnelTransition, parameterName=Speed Lines Color, defaultValue=<rgba 0.0036 10 0 1>}.RGB
   B=Multiply(A=Multiply(A=@n21.R, B=@n66), B=ConstantBiasScale(None=@n67))
  )
  n69 = MaterialFunctionCall() {materialFunction=&LinearGradient}
  n70 = Multiply(A=SmoothStep(Value=@n69.VGradient) {constMin=0.5}, B=OneMinus(None=SmoothStep(Value=@n69.VGradient) {constMin=0.8}))
  n71 = OneMinus(None=Saturate(None=Multiply(A=@n70) {constB=1.5}))
  n72 = OneMinus(None=MaterialFunctionCall() {materialFunction=&LinearGradient}.VGradient)
  NamedRerouteDeclaration(None=Multiply(A=Multiply(A=@n68, B=@n71), B=SmoothStep(Value=@n72) {constMax=0.3})) {name=EmissionLineColor, nodeColor=<rgba 0.125 0 1 1>}
  n73 = Constant()
  n74 = NamedRerouteDeclaration(None=ScalarParameter() {defaultValue=1, group=UVs, parameterName=UV U Mult02}) {name=UMulti02, nodeColor=<rgba 0.5 1 0 1>}
  n75 = NamedRerouteDeclaration(None=ScalarParameter() {defaultValue=1, group=UVs, parameterName=UV V Mult02}) {name=VMulti02, nodeColor=<rgba 1 0 0.8438 1>}
  n76 = NamedRerouteUsage()
  n77 = NamedRerouteUsage()
  n78 = NamedRerouteDeclaration(None=Divide(A=NamedRerouteUsage(), B=Divide(A=@n76, B=@n77))) {name=USpeed02, nodeColor=<rgba 0 1 0.4844 1>}
  n79 = NamedRerouteUsage()
  n80 = NamedRerouteUsage()
  n81 = NamedRerouteDeclaration(None=Divide(A=NamedRerouteUsage(), B=Divide(A=@n79, B=@n80))) {name=Vspeed02, nodeColor=<rgba 0 0.5078 1 1>}
  n82 = TextureSample(
   UVs=Rotator(
    Coordinate=MaterialFunctionCall(UV Channel=@n73, U Freq Multiplier=@n74, V Freq Multiplier=@n75, Pan Speed U=@n78, Pan Speed V=@n81) {materialFunction=&MF_FX_TextureUVPanner}.Result
    Time=Constant() {r=4.7125}
   ) {speed=1}
   Tex=TextureObjectParameter() {group=TunnelTransition, parameterName=01_Graffiti02, texture=/SewerPortal/Gameplay/VFX/Texture/T_Bricks_Cracks_Graffiti_Layer}
  )
  n83 = ScalarParameter() {defaultValue=0.5, group=TunnelTransition, parameterName=GraffitiLayerOpacity}
  NamedRerouteDeclaration(None=Multiply(A=@n82.RGB, B=Multiply(A=@n82.A, B=@n83))) {name=GraffitiLayer01, nodeColor=<rgba 0 0.4141 1 1>}
  NamedRerouteUsage()
  ParticleColor()
  n84 = NamedRerouteUsage()
  n85 = Add(
   A=OneMinus(None=NamedRerouteUsage())
   B=LinearInterpolate(A=Constant() {r=-1}, B=Constant() {r=1}, Alpha=ScalarParameter() {group=TunnelTransition, parameterName=WhiteWashDebugger})
  )
  n86 = VectorParameter() {group=TunnelTransition, parameterName=TunnelExitColor, defaultValue=<rgba 1.3266 2 1.3021 1>}
  n87 = Add(
   A=Multiply(A=@n84, B=ScalarParameter() {defaultValue=1, group=TunnelTransition, parameterName=TunnelEmissiveMultiplierDebugger})
   B=Clamp(None=Multiply(A=@n85, B=@n86.RGB))
  )
  n88 = ScalarParameter() {defaultValue=1, group=TunnelTransition, parameterName=FinalEmissiveMultiplierDebugger}
  n89 = DynamicParameter() {paramNames [WhiteWash, Final Emissive Multiplier, Tunnel Emissive Multiplier, Param4]}
  n90 = Add(A=OneMinus(None=NamedRerouteUsage()), B=LinearInterpolate(A=Constant() {r=-1}, B=Constant() {r=1}, Alpha=@n89.WhiteWash))
  n91 = VectorParameter() {group=TunnelTransition, parameterName=TunnelExitColor, defaultValue=<rgba 1.3266 2 1.3021 1>}
  n92 = Add(A=Multiply(A=@n84, B=@n89.Tunnel Emissive Multiplier), B=Clamp(None=Multiply(A=@n90, B=@n91.RGB)))
  StaticSwitchParameter(
   True=MaterialFunctionCall(Emissive=Multiply(A=@n87, B=@n88)) {materialFunction=&MF_FX_EmissiveAdjustments}.Result
   False=MaterialFunctionCall(Emissive=Multiply(A=@n92, B=@n89.Final Emissive Multiplier)) {materialFunction=&MF_FX_EmissiveAdjustments}.Result
  ) {group=TunnelTransition, parameterName=00_DebugTunnelColor}
 ]
}
