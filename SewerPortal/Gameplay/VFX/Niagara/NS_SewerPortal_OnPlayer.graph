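# Niagara system for a short steam puff, named as the on-player effect for the sewer portal.
# One sprite emitter combines a small burst with a curve-driven spawn rate that falls to zero over 0.8 seconds, spawning around a sphere.
# Upward acceleration, drag and curl noise move the SubUV sprites as they grow, whiten and fade, with opacity scaled by time of day.
# Cone velocity and a player-driven point force are present but disabled.
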NiagaraSystem /SewerPortal/Gameplay/VFX/Niagara/NS_SewerPortal_OnPlayer
dependencies {
 usedDataInterfaces [&NiagaraDataInterfaceCurve]
 usedDynamicInputs [&self:ScratchDynamicInput]
 usedModules [
  &self:SystemSpawnScript.SystemScriptSource.SystemScriptGraph.NiagaraNodeAssignment_3.SetVariables_#1
  &self:SystemSpawnScript.SystemScriptSource.SystemScriptGraph.NiagaraNodeAssignment_5.SetVariables_#2
  &self:SystemSpawnScript.SystemScriptSource.SystemScriptGraph.NiagaraNodeAssignment_2.SetVariables_#3
  &SystemState
  &self:SystemSpawnScript.SystemScriptSource.SystemScriptGraph.NiagaraNodeAssignment_0.SetVariables_#4
  &self:SystemSpawnScript.SystemScriptSource.SystemScriptGraph.NiagaraNodeAssignment_1.SetVariables_#5
 ]
}
emitters {Steam {
  emitterUpdateScript [
   &EmitterState(Life Cycle Mode=::NewEnumerator1, Loop Behavior=::NewEnumerator1, Loop Duration=0.8, UseLoopDelay=false)
   &SpawnBurst_Instantaneous(Spawn Count=2, Use Spawn Probability=false, UseLoopCountLimit=false)
   &SpawnRate(
    SpawnRate=&FloatFromCurve(
     FloatCurve=DataInterface() {propertyValues {Curve {keys [
         {arriveTangent=-2.3042, interpMode=RCIM_Cubic, leaveTangent=-2.3042, tangentMode=RCTM_User, time=0, value=1}
         {interpMode=RCIM_Cubic, tangentMode=RCTM_User, time=0.8, value=0}
        ]}}}
     CurveIndex=$Emitter.NormalizedLoopAge
     Scale Curve=50
    )
    Use Spawn Probability=false
   )
  ]
  particleSpawnScript [
   &InitializeParticle(
    Lifetime Max=1
    Lifetime Min=0.5
    Lifetime Mode=::NewEnumerator1
    Position Offset=&Add_Vector(A=<0 0 0>, B=$System.Location Offset)
    Sprite Rotation Mode=::NewEnumerator2
    Sprite Size Mode=::NewEnumerator1
    Uniform Sprite Size Max=160
    Uniform Sprite Size Min=140
    UsePositionOffset=true
    Write Material Random=true
   )
   &ShapeLocation(MaskSpawnGroup=false, Override Seed=false, Sphere Radius=40, UseOwnerScale=true)
   &AddVelocity(
    Cone Angle=30
    Cone Axis=<0 0 1>
    Module.UseConeFalloff=false
    Module.UseVelDistribution=false
    Override Seed=false
    Velocity Mode=::NewEnumerator2
    Velocity Speed=&RandomRangeFloat(Minimum=400, Maximum=600)
   ) {enabled=false, module=AddVelocity001}
   &CameraOffset()
  ]
  particleUpdateScript [
   &ParticleState()
   &AccelerationForce(Acceleration=&RandomRangeVector(Minimum=<0 0 100>, Maximum=<0 0 200>))
   &PointForce(
    Force Falloff Distance=90
    Force Origin=$System.PlayerPosition
    Force Strength=350
    Linear Force Vector:Vector3f=$System.PlayerVelocity
    Module.Use Falloff=true
    Module.UseLinearForce=true
   ) {enabled=false}
   &Drag(Drag=5, Use Linear Drag=true, Use Rotational Drag=false)
   &CurlNoiseForce(Noise Frequency=100, Noise Strength=100, Pan Noise=true, Pan Noise Field=<0 0 1>)
   &SolveForcesAndVelocity(Clamp Velocity=false, Limit Acceleration=false, Speed Limit=100)
   &Color(Color=&MakeLinearColorFromVectorAndFloat(
     "Vector (RGB)"=&Lerp_Vector(
      A=&MakeVectorFromLinearColorRGB(Color=<rgba 1 0.7284 0.4247 1>)
      B=&MakeVectorFromLinearColorRGB(Color=<rgba 1 1 1 1>)
      Alpha=&FloatFromCurve(FloatCurve=DataInterface() {propertyValues {Curve {keys [
           {interpMode=RCIM_Cubic, tangentMode=RCTM_User, time=0.5, value=0}
           {arriveTangent=1.6109, interpMode=RCIM_Cubic, leaveTangent=1.6109, tangentMode=RCTM_User, time=1, value=1}
          ]}}})
     )
     "Float (Alpha)"=&FloatFromCurve(
      FloatCurve=DataInterface() {propertyValues {Curve {keys [
          {arriveTangent=3.8562, interpMode=RCIM_Cubic, leaveTangent=3.8562, tangentMode=RCTM_User, time=0, value=0.6}
          {arriveTangent=-0.0041, interpMode=RCIM_Cubic, leaveTangent=-0.0041, tangentMode=RCTM_User, time=0.2, value=1}
          {arriveTangent=-1.6709, interpMode=RCIM_Cubic, leaveTangent=-1.6709, tangentMode=RCTM_User, time=1, value=0}
         ]}}}
      Scale Curve=&Multiply_Float(
       A=0.4
       B=&FNTimeOfDayScalar(Night Scale=0.15, Morning Scale=0.2, Day Scale=0.2, Evening Scale=0.3)
      )
     )
    ))
   &ScaleSpriteSize(Uniform Curve Sprite Scale=DataInterface() {propertyValues {Curve {keys [
        {arriveTangent=1.0548, interpMode=RCIM_Cubic, leaveTangent=1.0548, tangentMode=RCTM_User, time=0, value=0.6}
        {arriveTangent=0.0294, interpMode=RCIM_Cubic, leaveTangent=0.0294, tangentMode=RCTM_User, time=1, value=1.2}
       ]}}})
   &SubUVAnimation(Start Frame:NiagaraInt32=&RandomRangeInt(Minimum=0, Maximum=10))
  ]
  rendererClasses [&NiagaraSpriteRendererProperties]
  renderers [{rendererClass=&NiagaraSpriteRendererProperties}]
 }}
summary {
 systemName=NS_SewerPortal_OnPlayer
 emitters [{emitterName=Steam, rendererClasses [&NiagaraSpriteRendererProperties]}]
 userVariables [
  User.bIsFrontEnd {type=NiagaraBool}
  User.bIsFrontEndPreview {type=NiagaraBool}
 ]
}
